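# UI kit download progress indicator widget: a material-driven progress image with a centered download icon.
# PreConstruct sizes the image from WidgetSize and the icon at half that.
# Construct caches the dynamic material, which SetProgress feeds with negatives floored to zero and SetFeedbackVisibility uses to toggle its Display Feedback parameter.
# Anim_ArrowLoop defines a one-second vertical icon bob, but no graph here plays it.
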WidgetBlueprint /FortUILibrary/UIKit/ProgressBar/WBP_UI_DownloadProgress
animations [Anim_ArrowLoop {
  binding_count=1
  length_seconds=1
  object=&self:Anim_ArrowLoop
  sequence {
   bindings [Img_DownloadIcon {
     id=#1
     possessed_class=Image
     tracks [{
       type=MovieScene2DTransformTrack
       sections [{
         type=MovieScene2DTransformSection
         end_frame=60
         start_frame=0
         channels [
          Translation.X {type=MovieSceneScriptingFloatChannel}
          Translation.Y {type=MovieSceneScriptingFloatChannel, keys [[0,0],[18,-3],[36,0],[60,0]]}
          Angle {type=MovieSceneScriptingFloatChannel}
          Scale.X {type=MovieSceneScriptingFloatChannel}
          Scale.Y {type=MovieSceneScriptingFloatChannel}
          Shear.X {type=MovieSceneScriptingFloatChannel}
          Shear.Y {type=MovieSceneScriptingFloatChannel}
         ]
        }]
      }]
    }]
   display_rate {numerator=60}
  }
 }]
graphs [
 EventGraph {nodes [
   Event PreConstruct()
   n1 = Get WidgetSize() -> (WidgetSize:Vector 2D Structure)
   > SetDesiredSizeOverride!(execute, self:Image Object Reference=Get ProgressBar() -> (ProgressBar:Image Object Reference), DesiredSize:Vector 2D Structure=@n1) -> (then)
   > SetDesiredSizeOverride(
    self=Get Img_DownloadIcon() -> (Img_DownloadIcon:Image Object Reference)
    DesiredSize=vector2d * vector2d(A:Vector 2D Structure=@n1, B:Vector 2D Structure=(X=0.5,Y=0.5)) -> (ReturnValue:Vector 2D Structure) {is=title K2Node_PromotableOperator}
   )
   Event Construct()
   > n2 = GetDynamicMaterial(self=Get ProgressBar())
   > Set DynamicProgressMaterial(DynamicProgressMaterial:Material Instance Dynamic Object Reference=@n2) -> (Output_Get:Material Instance Dynamic Object Reference)
  ]}
 SetFeedbackVisibility {nodes [
   n1 = SetFeedbackVisibility!() -> (then, IsVisible:bool=~) {is=title K2Node_FunctionEntry}
   > SetScalarParameterValue(
    self=Get DynamicProgressMaterial() -> (DynamicProgressMaterial:Material Instance Dynamic Object Reference)
    ParameterName=Display Feedback
    Value=Select(Option 0:double=0, Option 1:double=1, Index=@n1) -> (ReturnValue:double)
   )
  ]}
 SetProgress {nodes [
   n1 = SetProgress!() -> (then, Value:double=~) {is=title K2Node_FunctionEntry}
   > SetScalarParameterValue(self=Get DynamicProgressMaterial(), ParameterName=Progress, Value=Select(Option 0=@n1, Option 1=0, Index="float < float"(A=@n1)))
  ]}
]
variables [
 DynamicProgressMaterial {type=MaterialInstanceDynamic}
 WidgetSize {type=Vector2D, default=<48 48>}
]
widgets {
 info {inheritedWidgetCount=0, parentClass=&UserWidget, rootWidgetClass=&Overlay, widgetCount=3}
 widgets [
  Overlay_DownloadProgress {widgetClassPath=&Overlay, bIsVariable=true, widget=Overlay_DownloadProgress}
  ProgressBar {
   widgetClassPath=&Image
   bIsVariable=true
   parent=Overlay_DownloadProgress
   widget=ProgressBar
   properties {
    visibility=HitTestInvisible
    brush {
     bIsDynamicallyLoaded=false
     drawAs=Image
     imageType=FullColor
     mirroring=NoMirror
     resourceName=None
     resourceObject=/FortUILibrary/Textures/MaterialInstances/MI_UI_DownloadProgress
     tiling=NoTile
     imageSize=<48 48>
     margin=<ltrb 0 0 0 0>
     outlineSettings {
      bUseBrushTransparency=false
      roundingType=HalfHeightRadius
      width=0
      color {colorUseRule=UseColor_Specified, specifiedColor=<rgba 0 0 0 0>}
      cornerRadii=<0 0 0 0>
     }
     tintColor {colorUseRule=UseColor_Specified, specifiedColor=<rgba 1 1 1 1>}
     uVRegion {bIsValid=false, max=<0 0>, min=<0 0>}
    }
   }
   slot {class=OverlaySlot, properties {horizontalAlignment=HAlign_Fill, verticalAlignment=VAlign_Fill}}
  }
  Img_DownloadIcon {
   widgetClassPath=&Image
   bIsVariable=true
   parent=Overlay_DownloadProgress
   widget=Img_DownloadIcon
   properties {
    visibility=HitTestInvisible
    brush {
     bIsDynamicallyLoaded=false
     drawAs=Image
     imageType=FullColor
     mirroring=NoMirror
     resourceName=None
     resourceObject=/FortUILibrary/Textures/MaterialInstances/MI_UIKit_Icon_Download
     tiling=NoTile
     imageSize=<24 24>
     margin=<ltrb 0 0 0 0>
     outlineSettings {
      bUseBrushTransparency=false
      roundingType=HalfHeightRadius
      width=0
      color {colorUseRule=UseColor_Specified, specifiedColor=<rgba 0 0 0 0>}
      cornerRadii=<0 0 0 0>
     }
     tintColor {colorUseRule=UseColor_Specified, specifiedColor=<rgba 1 1 1 1>}
     uVRegion {bIsValid=false, max=<0 0>, min=<0 0>}
    }
   }
   slot {class=OverlaySlot, properties {horizontalAlignment=HAlign_Center, verticalAlignment=VAlign_Center}}
  }
 ]
}
